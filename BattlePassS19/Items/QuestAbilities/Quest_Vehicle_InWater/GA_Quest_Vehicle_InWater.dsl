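; Quest gameplay ability that credits the owning player's quest when their vehicle ends up in water.
; A driving event on a vehicle matching the valet tag starts a two-second timer that caches the vehicle and checks for water.
; A vehicle exit event switches to a timer checking the cached vehicle's water state and forward speed.
; Detected water sends a custom stat event to the quest manager.
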(event Custom|Check%20In%20Water
  (bind _asfort_athena_vehicle (Utilities|Casting|CastToFortAthenaVehicle (Vehicle|GetVehicleInterface (Variables|GAVars|Data|GetOwningPlayerPawn)))
    (:then
      (Variables|Default|SetFort%20Athena%20Vehicle _asfort_athena_vehicle)
      (if (Class|FortVehicleInterface|IsinWater (Vehicle|GetVehicleInterface (Variables|GAVars|Data|GetOwningPlayerPawn)))
        (Development|PrintString "@@@@@@ ActivateAbilityFromEvent [GA_Quest_Vehicle_InWater]: IS IN WATER, sending submerged Custom Stat Event..." false (Variables|Debug|GetbDebugAbility))
        (Quest|SendComplexCustomStatEvent (Variables|GAVars|Data|GetOwningPlayerQuestManager) 0 (Variables|Default|GetQuestStatEventTags) "(GameplayTags=)" 1)
        (Utilities|Time|ClearandInvalidateTimerbyHandle (Variables|Default|GetTimer%20Handle))
        (else
          (Development|PrintString "@@@@@@ ActivateAbilityFromEvent [GA_Quest_Vehicle_InWater]: IS NOT IN WATER, waiting to check again..." false (Variables|Debug|GetbDebugAbility)))))
    (:CastFailed
      _))
  (Ability|EndAbility))

(event Ability|EventActivateAbilityFromEvent
  (params (EventData :type Gameplay%20Event%20Data%20Structure%20%28by%20ref%29))
  (|CallParentFunction EventData)
  (switch Utilities|FlowControl|Switch|SwitchonGameplayTag (Utilities|Struct|BreakGameplayEventData EventData)
    (:Event.Vehicle.Exit
      (Development|PrintString "@@@ ActivateAbilityFromEvent [GA_Quest_Vehicle_InWater]: CALLED; Vehicle Exited!" false (Variables|Debug|GetbDebugAbility))
      (CallFunction|ClearTimer (Variables|Default|GetTimer%20Handle))
      (bind _returnvalue (Utilities|Time|SetTimerbyEvent (delegate Custom|Check%20Speed) 2.0 true))
      (Variables|Default|SetTimer%20Handle _returnvalue))
    (:MovementMode.Driving
      (Development|PrintString "@@@ ActivateAbilityFromEvent [GA_Quest_Vehicle_InWater]: CALLED; Driving Vehicle!" false (Variables|Debug|GetbDebugAbility))
      ;[1 1 1] Check Vehicle Tags
      (if (GameplayTags|TagContainer|CompareTagAssetInterfacetoOtherTags (Variables|Default|GetValetTag) (Vehicle|GetVehicleAbilitySystemComponent (Vehicle|GetVehicleInterface (Variables|GAVars|Data|GetOwningPlayerPawn))))
        ;[1 1 1] Start Checking Vehicle If Vehicle is in Water
        (bind _returnvalue_2 (Utilities|Time|SetTimerbyEvent (delegate Custom|Check%20In%20Water) 2.0 true))
        (Variables|Default|SetTimer%20Handle _returnvalue_2)
        (else
          (Ability|EndAbility))))))

(event Custom|Check%20Speed
  (bind _fort_athena_vehicle (Variables|Default|GetFort%20Athena%20Vehicle))
  (bind _returnvalue (Vehicle|GetForwardSpeedKmh _fort_athena_vehicle))
  (bind _returnvalue_2 (Class|FortAthenaVehicle|IsinWaterBody _fort_athena_vehicle))
  (Development|PrintString (Utilities|String|Append "@@@@@@ ActivateAbilityFromEvent [GA_Quest_Vehicle_InWater]: SPEED = " (Utilities|String|ToString(Float) _returnvalue) "; IS IN WATER =  " (Utilities|String|ToString(Boolean) _returnvalue_2)) true (Variables|Debug|GetbDebugAbility))
  ;[1 1 1] Check if in water...
  (if _returnvalue_2
    (Quest|SendComplexCustomStatEvent (Variables|GAVars|Data|GetOwningPlayerQuestManager) 0 (Variables|Default|GetQuestStatEventTags) "(GameplayTags=)" 1)
    (elif (<= _returnvalue 0.0)
      _))
  (CallFunction|ClearTimer (Variables|Default|GetTimer%20Handle))
  (Ability|EndAbility))

(fn ClearTimer
  (params (Timer :type Timer%20Handle%20Structure))
  (if (Utilities|Time|IsValidTimerHandle Timer)
    (Utilities|Time|ClearandInvalidateTimerbyHandle Timer)
    (else
      _))
  (return))
